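# Physics asset for the left-side mesh of the BullKeynote medium male player cosmetic body, built around three skirt bone chains (skirt, side, back).
# Each chain has a thin base capsule angularly free on the pelvis, followed by three capsule links with tight swing and zero twist limits, the side chain swinging on one axis only.
# Capsules on pelvis, spine, left leg and, notably, right forearm and hand provide the surrounding body shapes.
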PhysicsAsset /BRCosmetics/Characters/Player/Male/Medium/Bodies/M_MED_BullKeynote/Meshes/M_MED_BullKeynoteLeft_Physics
bodies {
 base_skirtBack_L [{shapeType=Capsule, length=16.505, radius=0.505, shapeName=base_skirtBack_L_capsule, rotation=<rot 0 -90 180>}]
 base_skirtSide_L [{shapeType=Capsule, length=16.505, radius=0.505, shapeName=base_skirtSide_L_capsule, rotation=<rot 0 -90 180>}]
 base_skirt_L [{shapeType=Capsule, length=16.505, radius=0.505, shapeName=base_skirt_L_capsule, rotation=<rot 0 -90 180>}]
 calf_l [{
   shapeType=Capsule
   length=39.1145
   radius=6.5924
   shapeName=calf_l_capsule
   center=<-26.3868 -0.1775 1.0013>
   rotation=<rot -2.134 -91.6359 91.6064>
  }]
 dyn_skirtBack_L_01 [{shapeType=Capsule, length=24, radius=6, center=<10.9877 -2.4143 3.3146>, rotation=<rot 0 0 -180>}]
 dyn_skirtBack_L_02 [{shapeType=Capsule, length=28.4377, radius=5.3368, shapeName=dyn_skirtBack_R_02_capsule, center=<8.5891 -1.5721 0>, rotation=<rot 0 0 180>}]
 dyn_skirtBack_L_03 [{shapeType=Capsule, length=28.9377, radius=5.8368, shapeName=dyn_skirtBack_R_03_capsule, center=<7.0709 -2.9002 0>, rotation=<rot 0 0 -180>}]
 dyn_skirtSide_L_01 [{shapeType=Capsule, length=21, radius=3, center=<9.369 0 0>, rotation=<rot 90 -90 0>}]
 dyn_skirtSide_L_02 [{shapeType=Capsule, length=36, radius=5, shapeName=dyn_skirtSide_L_02_capsule, center=<6.977 0 2.1396>, rotation=<rot 90 -90 0>}]
 dyn_skirtSide_L_03 [{shapeType=Capsule, length=37, radius=4, shapeName=dyn_skirtSide_L_03_capsule, center=<7.862 0 2.1396>, rotation=<rot 90 -90 0>}]
 dyn_skirt_L_01 [{shapeType=Capsule, length=30, radius=3.5, center=<10.6671 1.4682 -2.5756>, rotation=<rot 18.7918 90.0005 180>}]
 dyn_skirt_L_02 [{shapeType=Capsule, length=28, radius=5, shapeName=dyn_skirt_L_02_capsule, center=<6.9727 4.0032 -1.713>, rotation=<rot 19 90 180>}]
 dyn_skirt_L_03 [{
   shapeType=Capsule
   length=30.5
   radius=4
   shapeName=dyn_skirt_L_03_capsule
   center=<8.2354 1.4682 -2.5756>
   rotation=<rot 18.7918 90.0005 180>
  }]
 foot_l [{shapeType=Capsule, length=9.9831, radius=6.9581, shapeName=foot_l_capsule, center=<-3.9988 2.8657 0.2258>, rotation=<rot 0 12.2368 90.6957>}]
 hand_r [{shapeType=Capsule, length=8.8136, radius=5.5079, center=<-8.0147 2.5711 0.2707>, rotation=<rot 0 81.2387 -76.3593>}]
 lowerarm_r [{shapeType=Capsule, length=11.3324, radius=4.6021, center=<-12.2313 -0.1782 0.0265>, rotation=<rot 0 -90.6108 -89.2979>}]
 pelvis [{shapeType=Capsule, length=17.3556, radius=8.4332, center=<7.4792 0.6148 0.6791>, rotation=<rot 0 -90 0>}]
 spine_02 [{shapeType=Capsule, length=12.3413, radius=8.6866, center=<9.2478 1.8989 -0.0563>, rotation=<rot 6.6164 -90 -0.5562>}]
 thigh_l [{shapeType=Capsule, length=39.0063, radius=5.5764, center=<-19.1958 0.7954 1.4639>, rotation=<rot 0 -87.77 91.8661>}]
}
constraints [
 {bone1Name=spine_02, bone2Name=pelvis, swing1LimitDegrees=45, swing2LimitDegrees=45, twistLimitDegrees=45}
 {bone1Name=lowerarm_r, bone2Name=spine_02, swing1LimitDegrees=45, swing2LimitDegrees=45, twistLimitDegrees=45}
 {bone1Name=hand_r, bone2Name=lowerarm_r, swing1LimitDegrees=45, swing2LimitDegrees=45, twistLimitDegrees=45}
 {bone1Name=dyn_skirtBack_L_01, bone2Name=base_skirtBack_L, swing1LimitDegrees=10, swing2LimitDegrees=5, twistLimitDegrees=0}
 {bone1Name=dyn_skirtBack_L_02, bone2Name=dyn_skirtBack_L_01, swing1LimitDegrees=10, swing2LimitDegrees=5, twistLimitDegrees=0}
 {bone1Name=dyn_skirtBack_L_03, bone2Name=dyn_skirtBack_L_02, swing1LimitDegrees=10, swing2LimitDegrees=5, twistLimitDegrees=0}
 {
  bone1Name=base_skirt_L
  bone2Name=pelvis
  swing1LimitDegrees=45
  swing1Motion=Free
  swing2LimitDegrees=45
  swing2Motion=Free
  twistLimitDegrees=45
  twistMotion=Free
 }
 {bone1Name=dyn_skirt_L_01, bone2Name=base_skirt_L, swing1LimitDegrees=10, swing2LimitDegrees=5, twistLimitDegrees=0}
 {bone1Name=dyn_skirt_L_02, bone2Name=dyn_skirt_L_01, swing1LimitDegrees=10, swing2LimitDegrees=5, twistLimitDegrees=0}
 {bone1Name=dyn_skirt_L_03, bone2Name=dyn_skirt_L_02, swing1LimitDegrees=10, swing2LimitDegrees=5, twistLimitDegrees=0}
 {
  bone1Name=base_skirtBack_L
  bone2Name=pelvis
  swing1LimitDegrees=45
  swing1Motion=Free
  swing2LimitDegrees=45
  swing2Motion=Free
  twistLimitDegrees=45
  twistMotion=Free
 }
 {bone1Name=dyn_skirtSide_L_03, bone2Name=dyn_skirtSide_L_02, swing1LimitDegrees=0, swing2LimitDegrees=20, twistLimitDegrees=0}
 {bone1Name=dyn_skirtSide_L_02, bone2Name=dyn_skirtSide_L_01, swing1LimitDegrees=0, swing2LimitDegrees=20, twistLimitDegrees=0}
 {bone1Name=dyn_skirtSide_L_01, bone2Name=base_skirtSide_L, swing1LimitDegrees=0, swing2LimitDegrees=10, twistLimitDegrees=0}
 {
  bone1Name=base_skirtSide_L
  bone2Name=pelvis
  swing1LimitDegrees=45
  swing1Motion=Free
  swing2LimitDegrees=45
  swing2Motion=Free
  twistLimitDegrees=45
  twistMotion=Free
 }
 {bone1Name=thigh_l, bone2Name=pelvis, swing1LimitDegrees=45, swing2LimitDegrees=45, twistLimitDegrees=45}
 {
  bone1Name=calf_l
  bone2Name=thigh_l
  swing1LimitDegrees=45
  swing1Motion=Free
  swing2LimitDegrees=45
  swing2Motion=Free
  twistLimitDegrees=45
  twistMotion=Free
 }
 {
  bone1Name=foot_l
  bone2Name=calf_l
  swing1LimitDegrees=45
  swing1Motion=Free
  swing2LimitDegrees=45
  swing2Motion=Free
  twistLimitDegrees=45
  twistMotion=Free
 }
]
