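# Physics asset for the right side (R) of the BullKeynote medium male player body cosmetic, built to drive skirt motion.
# Three three-link skirt chains (main, side, back) hang from thin base capsules, which attach to the pelvis with free rotation.
# Chain joints allow only small swings with zero twist, mostly on a single swing axis.
# Capsules on pelvis, spine and right forearm, hand, leg and foot form a partial body alongside the chains.
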PhysicsAsset /BRCosmetics/Characters/Player/Male/Medium/Bodies/M_MED_BullKeynote/Meshes/M_MED_BullKeynoteR_Physics
bodies {
 base_skirtBack_R [{shapeType=Capsule, length=16.505, radius=0.505, shapeName=base_skirtBack_R_capsule, rotation=<rot 0 -90 0>}]
 base_skirtSide_R [{shapeType=Capsule, length=16.505, radius=0.505, shapeName=base_skirtSide_R_capsule, rotation=<rot 0 -90 0>}]
 base_skirt_R [{shapeType=Capsule, length=16.505, radius=0.505, shapeName=base_skirt_R_capsule, rotation=<rot 0 -90 0>}]
 calf_r [{shapeType=Capsule, length=43.1097, radius=7.2365, shapeName=calf_r_capsule, center=<23.4631 -0.3463 -1.1705>, rotation=<rot 90 0 0>}]
 dyn_skirtBack_R_01 [{shapeType=Capsule, length=18, radius=6, center=<-10.9877 2.4143 -3.3146>}]
 dyn_skirtBack_R_02 [{shapeType=Capsule, length=29.4377, radius=5.3368, shapeName=dyn_skirtBack_L_02_capsule, center=<-8.5891 1.5721 0>}]
 dyn_skirtBack_R_03 [{shapeType=Capsule, length=27.9377, radius=4.5, shapeName=dyn_skirtBack_L_03_capsule, center=<-8.3121 3.9231 0>}]
 dyn_skirtSide_R_01 [{shapeType=Capsule, length=21, radius=3, center=<-9.369 0 0>, rotation=<rot 90 90 0>}]
 dyn_skirtSide_R_02 [{shapeType=Capsule, length=36.5, radius=4.5, shapeName=dyn_skirtSide_R_02_capsule, center=<-6.977 0 -2.1396>, rotation=<rot 90 90 0>}]
 dyn_skirtSide_R_03 [{shapeType=Capsule, length=37, radius=4, shapeName=dyn_skirtSide_R_03_capsule, center=<-7.862 0 -2.1396>, rotation=<rot 90 90 0>}]
 dyn_skirt_R_01 [{shapeType=Capsule, length=21, radius=3.5, center=<-10.6671 0 0>, rotation=<rot 20 90 0>}]
 dyn_skirt_R_02 [{shapeType=Capsule, length=19, radius=5, shapeName=dyn_skirt_R_02_capsule, center=<-6.9727 -1.1538 -5.8589>, rotation=<rot 20 90 0>}]
 dyn_skirt_R_03 [{shapeType=Capsule, length=21.5, radius=4, shapeName=dyn_skirt_R_03_capsule, center=<-8.2354 0 -5.7107>, rotation=<rot 20 90.0001 0>}]
 foot_r [{shapeType=Capsule, length=9.7867, radius=5.3034, shapeName=foot_r_capsule, center=<4.718 -2.3806 -0.177>, rotation=<rot 0 11.6161 -87.5596>}]
 hand_r [{shapeType=Capsule, length=8.8136, radius=5.5079, center=<-8.0147 2.5711 0.2707>, rotation=<rot 0 81.2387 -76.3593>}]
 lowerarm_r [{shapeType=Capsule, length=11.3324, radius=4.6021, center=<-12.2313 -0.1782 0.0265>, rotation=<rot 0 -90.6108 -89.2979>}]
 pelvis [{shapeType=Capsule, length=17.3556, radius=8.4332, center=<7.4792 0.6148 0.6791>, rotation=<rot 0 -90 0>}]
 spine_02 [{shapeType=Capsule, length=12.3413, radius=8.6866, center=<9.2478 1.8989 -0.0563>, rotation=<rot 6.6164 -90 -0.5562>}]
 thigh_r [{shapeType=Capsule, length=39.0063, radius=5.5764, center=<19.2194 -0.7945 -0.7385>, rotation=<rot 0 -87.7701 -90.0577>}]
}
constraints [
 {bone1Name=spine_02, bone2Name=pelvis, swing1LimitDegrees=45, swing2LimitDegrees=45, twistLimitDegrees=45}
 {bone1Name=lowerarm_r, bone2Name=spine_02, swing1LimitDegrees=45, swing2LimitDegrees=45, twistLimitDegrees=45}
 {bone1Name=hand_r, bone2Name=lowerarm_r, swing1LimitDegrees=45, swing2LimitDegrees=45, twistLimitDegrees=45}
 {bone1Name=thigh_r, bone2Name=pelvis, swing1LimitDegrees=45, swing2LimitDegrees=45, twistLimitDegrees=45}
 {bone1Name=dyn_skirtBack_R_01, bone2Name=base_skirtBack_R, swing1LimitDegrees=0, swing2LimitDegrees=10, twistLimitDegrees=0}
 {bone1Name=dyn_skirtSide_R_01, bone2Name=base_skirtSide_R, swing1LimitDegrees=0, swing2LimitDegrees=10, twistLimitDegrees=0}
 {bone1Name=dyn_skirt_R_01, bone2Name=base_skirt_R, swing1LimitDegrees=0, swing2LimitDegrees=10, twistLimitDegrees=0}
 {
  bone1Name=base_skirtBack_R
  bone2Name=pelvis
  swing1LimitDegrees=45
  swing1Motion=Free
  swing2LimitDegrees=45
  swing2Motion=Free
  twistLimitDegrees=45
  twistMotion=Free
 }
 {bone1Name=dyn_skirtBack_R_02, bone2Name=dyn_skirtBack_R_01, swing1LimitDegrees=10, swing2LimitDegrees=5, twistLimitDegrees=0}
 {bone1Name=dyn_skirtBack_R_03, bone2Name=dyn_skirtBack_R_02, swing1LimitDegrees=10, swing2LimitDegrees=5, twistLimitDegrees=0}
 {
  bone1Name=base_skirtSide_R
  bone2Name=pelvis
  swing1LimitDegrees=45
  swing1Motion=Free
  swing2LimitDegrees=45
  swing2Motion=Free
  twistLimitDegrees=45
  twistMotion=Free
 }
 {bone1Name=dyn_skirtSide_R_02, bone2Name=dyn_skirtSide_R_01, swing1LimitDegrees=0, swing2LimitDegrees=20, twistLimitDegrees=0}
 {bone1Name=dyn_skirtSide_R_03, bone2Name=dyn_skirtSide_R_02, swing1LimitDegrees=0, swing2LimitDegrees=20, twistLimitDegrees=0}
 {
  bone1Name=base_skirt_R
  bone2Name=pelvis
  swing1LimitDegrees=45
  swing1Motion=Free
  swing2LimitDegrees=45
  swing2Motion=Free
  twistLimitDegrees=45
  twistMotion=Free
 }
 {bone1Name=dyn_skirt_R_02, bone2Name=dyn_skirt_R_01, swing1LimitDegrees=0, swing2LimitDegrees=15, twistLimitDegrees=0}
 {bone1Name=dyn_skirt_R_03, bone2Name=dyn_skirt_R_02, swing1LimitDegrees=0, swing2LimitDegrees=15, twistLimitDegrees=0}
 {
  bone1Name=calf_r
  bone2Name=thigh_r
  swing1LimitDegrees=45
  swing1Motion=Free
  swing2LimitDegrees=45
  swing2Motion=Free
  twistLimitDegrees=45
  twistMotion=Free
 }
 {
  bone1Name=foot_r
  bone2Name=calf_r
  swing1LimitDegrees=45
  swing1Motion=Free
  swing2LimitDegrees=45
  swing2Motion=Free
  twistLimitDegrees=45
  twistMotion=Free
 }
]
