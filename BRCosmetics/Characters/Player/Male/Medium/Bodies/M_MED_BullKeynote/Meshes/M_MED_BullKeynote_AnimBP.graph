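# Animation Blueprint for the M_MED_BullKeynote male body cosmetic, adding secondary skirt motion to the pose copied from the parent mesh.
# A Control Rig receives crouch, emote, slide, melee and front-end states, then bone offsets shape the skirt around per-leg RigidBody simulations in thigh space.
# Constraints average center bones from left/right chains and tie back bones to calves.
# EventGraph converts character states to floats and computes gravity overrides.
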AnimBlueprint /BRCosmetics/Characters/Player/Male/Medium/Bodies/M_MED_BullKeynote/Meshes/M_MED_BullKeynote_AnimBP
target_skeleton=&/M_MED_BullKeynote_Skeleton
graphs [
 AnimGraph {nodes [
   n1 = Control Rig!(
    Source=Copy Pose From Mesh() {bUseAttachedParent=true}
    Alpha
    IsCrouching:bool=Get bIsCrouching()
    PlayinEmote:bool=Get bIsPlayingEmote()
    Kneeslide:double=Get Slidfe() -> (Slidfe:double)
    BBSlide:double=Get BBSlide() -> (BBSlide:double)
    Melee:double=Get Melee() -> (Melee:double)
    FE_Idle:double=Get FE() -> (FE:double)
   ) -> (Pose) {controlRigAssetReference {blueprintRigClass=&/M_MED_BullKeynote_CtrlRig.M_MED_BullKeynote_CtrlRig_C}}
   n2 = "Transform (Modify) Bone - Bone: base_skirtBack_L"(ComponentPose=Local To Component(LocalPose=@n1), Alpha=1, Translation="0,-5,0.5", Rotation="0,-15,-0.001", Scale="1,1,1") {
    rotationMode=BMM_Additive
    rotationSpace=BCS_BoneSpace
    translationMode=BMM_Additive
    translationSpace=BCS_BoneSpace
    alphaScaleBiasClamp {bInterpResult=true}
    boneToModify {boneName=base_skirtBack_L}
    rotation {roll=-0.001, yaw=-15}
    translation {y=-5, z=0.5}
   }
   n3 = "Transform (Modify) Bone - Bone: dyn_skirtBack_R_01"(ComponentPose=@n2, Alpha=1, Translation="0,3,0", Rotation="0,-15,0", Scale="1,1,1") {
    rotationMode=BMM_Additive
    rotationSpace=BCS_BoneSpace
    translationMode=BMM_Additive
    translationSpace=BCS_BoneSpace
    alphaScaleBiasClamp {bInterpResult=true}
    boneToModify {boneName=dyn_skirtBack_R_01}
    rotation {yaw=-15}
    translation {y=3}
   }
   n4 = "Transform (Modify) Bone - Bone: base_skirtBack_c_01"(ComponentPose=@n3, Alpha=1, Translation="0,0,1.5", Rotation="0,0,0", Scale="1,1,1") {
    translationMode=BMM_Additive
    translationSpace=BCS_BoneSpace
    alphaScaleBiasClamp {bInterpResult=true}
    boneToModify {boneName=base_skirtBack_c_01}
    translation {z=1.5}
   }
   n5 = "Transform (Modify) Bone - Bone: base_skirt_R"!(ComponentPose=@n4, bAlphaBoolEnabled:bool=True, Translation="0,0,0", Rotation="10.000,-9.000,45.989", Scale="1,1,1") -> (Pose) {
    alphaInputType=Bool
    rotationMode=BMM_Additive
    rotationSpace=BCS_BoneSpace
    alphaBoolBlend {blendInTime=0.2, blendOutTime=0.2}
    boneToModify {boneName=base_skirt_R}
    rotation=<rot 10 -9 45.9889>
   }
   n6 = "Transform (Modify) Bone - Bone: base_skirtSide_R"!(ComponentPose=@n5, bAlphaBoolEnabled:bool=True, Translation="0,0,0", Rotation="3.392,16.109,-10.032", Scale="1,1,1") -> (Pose) {
    alphaInputType=Bool
    rotationMode=BMM_Additive
    rotationSpace=BCS_BoneSpace
    alphaBoolBlend {blendInTime=0.2, blendOutTime=0.2}
    boneToModify {boneName=base_skirtSide_R}
    rotation=<rot 3.3919 16.109 -10.0321>
   }
   n7 = RigidBody!(ComponentPose=@n6, Alpha, OverrideWorldGravity:Vector="0,0,-980", ExternalForce) -> (Pose) {
    bClampLinearTranslationLimitToRefPose=true
    bOverrideWorldGravity=true
    overridePhysicsAsset=&/M_MED_BullKeynoteLeft_Physics
    simulationSpace=BaseBoneSpace
    alphaBoolBlend {blendOption=Linear}
    alphaScaleBiasClamp {bInterpResult=true}
    baseBoneRef {boneName=thigh_l}
    overrideWorldGravity {z=-980}
    simSpaceSettings {
     dampingAlpha=0
     maxAngularAcceleration=20
     maxAngularVelocity=5
     maxLinearAcceleration=2000
     maxLinearVelocity=400
     velocityScaleZ=0.3
     worldAlpha=0.4
     externalLinearDragV=<0.2 0.2 0.2>
    }
   }
   n8 = RigidBody(ComponentPose=@n7, OverrideWorldGravity="0,0,-980") {
    bClampLinearTranslationLimitToRefPose=true
    bOverrideWorldGravity=true
    overridePhysicsAsset=&/M_MED_BullKeynoteR_Physics
    simulationSpace=BaseBoneSpace
    alphaBoolBlend {blendOption=Linear}
    alphaScaleBiasClamp {bInterpResult=true}
    baseBoneRef {boneName=thigh_r}
    overrideWorldGravity {z=-980}
    simSpaceSettings {
     dampingAlpha=0
     maxAngularAcceleration=20
     maxAngularVelocity=5
     maxLinearAcceleration=2000
     maxLinearVelocity=400
     velocityScaleZ=0
     worldAlpha=0.4
     externalLinearDragV=<0.2 0.2 0.2>
    }
   }
   n9 = "Transform (Modify) Bone - Bone: dyn_skirtBack_R_02"!(ComponentPose=@n8, bAlphaBoolEnabled:bool=True, Translation="0,0,0", Rotation="0,-30,0", Scale="1,1,1") -> (Pose) {
    alphaInputType=Bool
    rotationMode=BMM_Additive
    rotationSpace=BCS_BoneSpace
    translationMode=BMM_Additive
    translationSpace=BCS_BoneSpace
    boneToModify {boneName=dyn_skirtBack_R_02}
    rotation {yaw=-30}
   }
   n10 = "Transform (Modify) Bone - Bone: dyn_skirt_L_01"!(ComponentPose=@n9, bAlphaBoolEnabled:bool=True, Translation="-1.273,1.5,-0.007", Rotation="14.621,1.144,0", Scale="1,1,1") -> (Pose) {
    alphaInputType=Bool
    rotationMode=BMM_Additive
    rotationSpace=BCS_BoneSpace
    translationMode=BMM_Additive
    translationSpace=BCS_BoneSpace
    boneToModify {boneName=dyn_skirt_L_01}
    rotation {pitch=14.621, yaw=1.144}
    translation=<-1.273 1.5 -0.007>
   }
   n11 = "Transform (Modify) Bone - Bone: dyn_skirt_L_02"!(ComponentPose=@n10, bAlphaBoolEnabled:bool=True, Translation="0,0,0", Rotation="-26.06,0,0", Scale="1,1,1") -> (Pose) {
    alphaInputType=Bool
    rotationMode=BMM_Additive
    rotationSpace=BCS_BoneSpace
    translationMode=BMM_Additive
    translationSpace=BCS_BoneSpace
    boneToModify {boneName=dyn_skirt_L_02}
    rotation {pitch=-26.06}
   }
   n12 = "Transform (Modify) Bone - Bone: dyn_skirtBack_R_01"!(ComponentPose=@n11, bAlphaBoolEnabled:bool=True, Translation="0,3,0", Rotation="23.882,-4.994,0", Scale="1,1,1") -> (Pose) {
    alphaInputType=Bool
    rotationMode=BMM_Additive
    rotationSpace=BCS_BoneSpace
    translationMode=BMM_Additive
    translationSpace=BCS_BoneSpace
    boneToModify {boneName=dyn_skirtBack_R_01}
    rotation {pitch=23.882, yaw=-4.994}
    translation {y=3}
   }
   n13 = "Transform (Modify) Bone - Bone: base_skirtBack_L"(ComponentPose=@n12, Alpha=1, Translation="0,0,0", Rotation="-20,10,5", Scale="1,1,1") {
    rotationMode=BMM_Additive
    rotationSpace=BCS_BoneSpace
    translationMode=BMM_Additive
    translationSpace=BCS_BoneSpace
    alphaScaleBiasClamp {bInterpResult=true, interpSpeedDecreasing=5, interpSpeedIncreasing=5}
    boneToModify {boneName=base_skirtBack_L}
    rotation=<rot -20 10 5>
   }
   n14 = "Transform (Modify) Bone - Bone: dyn_skirtBack_L_01"!(ComponentPose=@n13, bAlphaBoolEnabled:bool=True, Translation="0,-3,3", Rotation="0,0,0", Scale="1,1,1") -> (Pose) {
    alphaInputType=Bool
    rotationMode=BMM_Additive
    rotationSpace=BCS_BoneSpace
    translationMode=BMM_Additive
    translationSpace=BCS_BoneSpace
    boneToModify {boneName=dyn_skirtBack_L_01}
    translation {y=-3, z=3}
   }
   n15 = "Constraint - dyn_skirtSide_L_02 (RR)"!(ComponentPose=@n14, Alpha=1, ConstraintWeights_0=0.5, ConstraintWeights_1:float=0.5) -> (Pose) {
    boneToModify {boneName=dyn_skirtSide_L_02}
    constraintSetup [{transformType=Rotation, targetBone {boneName=dyn_skirt_L_02}}, {transformType=Rotation, targetBone {boneName=dyn_skirtBack_L_02}}]
    constraintWeights [1,1]
   }
   n16 = "Transform (Modify) Bone - Bone: dyn_skirtSide_R_01"!(ComponentPose=@n15, bAlphaBoolEnabled:bool=True, Translation="0,3,0", Rotation="5.562,15,0", Scale="1,1,1") -> (Pose) {
    alphaInputType=Bool
    rotationMode=BMM_Additive
    rotationSpace=BCS_BoneSpace
    translationMode=BMM_Additive
    translationSpace=BCS_BoneSpace
    boneToModify {boneName=dyn_skirtSide_R_01}
    rotation {pitch=5.562, yaw=15}
    translation {y=3}
   }
   n17 = "Transform (Modify) Bone - Bone: base_skirtSide_L"!(ComponentPose=@n16, bAlphaBoolEnabled:bool=True, Translation="0,-3,0", Rotation="12,0,0", Scale="1,1,1") -> (Pose) {
    alphaInputType=Bool
    rotationMode=BMM_Additive
    rotationSpace=BCS_BoneSpace
    translationMode=BMM_Additive
    translationSpace=BCS_BoneSpace
    boneToModify {boneName=base_skirtSide_L}
    rotation {pitch=12}
    translation {y=-3}
   }
   n18 = "Constraint - dyn_skirtBack_c_01 (RR)"!(ComponentPose=@n17, Alpha=1, ConstraintWeights_0=0.5, ConstraintWeights_1:float=0.5) -> (Pose) {
    alphaScaleBiasClamp {bInterpResult=true}
    boneToModify {boneName=dyn_skirtBack_c_01}
    constraintSetup [{transformType=Rotation, targetBone {boneName=dyn_skirtBack_L_01}}, {transformType=Rotation, targetBone {boneName=dyn_skirtBack_R_01}}]
    constraintWeights [1,1]
   }
   n19 = "Constraint - dyn_skirtBack_c_02 (RR)"!(ComponentPose=@n18, Alpha=1, ConstraintWeights_0=0.5, ConstraintWeights_1:float=0.5) -> (Pose) {
    alphaScaleBiasClamp {bInterpResult=true}
    boneToModify {boneName=dyn_skirtBack_c_02}
    constraintSetup [{transformType=Rotation, targetBone {boneName=dyn_skirtBack_L_02}}, {transformType=Rotation, targetBone {boneName=dyn_skirtBack_R_02}}]
    constraintWeights [1,1]
   }
   n20 = "Constraint - dyn_skirtBack_c_03 (RR)"!(ComponentPose=@n19, Alpha=1, ConstraintWeights_0=0.5, ConstraintWeights_1:float=0.5) -> (Pose) {
    alphaScaleBiasClamp {bInterpResult=true}
    boneToModify {boneName=dyn_skirtBack_c_03}
    constraintSetup [{transformType=Rotation, targetBone {boneName=dyn_skirtBack_L_03}}, {transformType=Rotation, targetBone {boneName=dyn_skirtBack_R_03}}]
    constraintWeights [1,1]
   }
   n21 = "Constraint - dyn_skirt_c_01 (RR)"!(ComponentPose=@n20, bAlphaBoolEnabled:bool=True, ConstraintWeights_0=0.5, ConstraintWeights_1:float=0.5) -> (Pose) {
    alphaInputType=Bool
    alphaBoolBlend {blendInTime=0.2, blendOption=Sinusoidal, blendOutTime=0.2}
    alphaScaleBiasClamp {bInterpResult=true}
    boneToModify {boneName=dyn_skirt_c_01}
    constraintSetup [{transformType=Rotation, targetBone {boneName=dyn_skirt_L_01}}, {transformType=Rotation, targetBone {boneName=dyn_skirt_R_01}}]
    constraintWeights [1,1]
   }
   n22 = "Constraint - dyn_skirt_c_02 (RR)"!(ComponentPose=@n21, bAlphaBoolEnabled:bool=True, ConstraintWeights_0=0.5, ConstraintWeights_1:float=0.5) -> (Pose) {
    alphaInputType=Bool
    alphaBoolBlend {blendInTime=0.2, blendOption=Sinusoidal, blendOutTime=0.2}
    alphaScaleBiasClamp {bInterpResult=true}
    boneToModify {boneName=dyn_skirt_c_02}
    constraintSetup [{transformType=Rotation, targetBone {boneName=dyn_skirt_L_02}}, {transformType=Rotation, targetBone {boneName=dyn_skirt_R_02}}]
    constraintWeights [1,1]
   }
   n23 = "Constraint - dyn_skirt_c_03 (RR)"!(ComponentPose=@n22, bAlphaBoolEnabled:bool=True, ConstraintWeights_0=0.5, ConstraintWeights_1:float=0.5) -> (Pose) {
    alphaInputType=Bool
    alphaBoolBlend {blendInTime=0.2, blendOption=Sinusoidal, blendOutTime=0.2}
    alphaScaleBiasClamp {bInterpResult=true}
    boneToModify {boneName=dyn_skirt_c_03}
    constraintSetup [{transformType=Rotation, targetBone {boneName=dyn_skirt_L_03}}, {transformType=Rotation, targetBone {boneName=dyn_skirt_R_03}}]
    constraintWeights [1,1]
   }
   n24 = "Constraint - dyn_skirtBack_c_02 (RR)"!(ComponentPose=@n23, bAlphaBoolEnabled:bool=True, ConstraintWeights_0=0.5, ConstraintWeights_1=0.5) -> (Pose) {
    alphaInputType=Bool
    alphaBoolBlend {blendInTime=0.4, blendOutTime=0.7}
    boneToModify {boneName=dyn_skirtBack_c_02}
    constraintSetup [{transformType=Rotation, targetBone {boneName=calf_r}}, {transformType=Rotation, targetBone {boneName=calf_l}}]
    constraintWeights [1,1]
   }
   n25 = "Constraint - dyn_skirtBack_R_02 (R)"!(ComponentPose=@n24, bAlphaBoolEnabled:bool=True, ConstraintWeights_0=1) -> (Pose) {
    alphaInputType=Bool
    alphaBoolBlend {blendInTime=0.4, blendOutTime=0.7}
    boneToModify {boneName=dyn_skirtBack_R_02}
    constraintSetup [{transformType=Rotation, targetBone {boneName=calf_r}}]
    constraintWeights [1]
   }
   n26 = "Constraint - dyn_skirtBack_L_02 (R)"!(ComponentPose=@n25, bAlphaBoolEnabled:bool=True, ConstraintWeights_0=1) -> (Pose) {
    alphaInputType=Bool
    alphaBoolBlend {blendInTime=0.4, blendOutTime=0.7}
    boneToModify {boneName=dyn_skirtBack_L_02}
    constraintSetup [{transformType=Rotation, targetBone {boneName=calf_l}}]
    constraintWeights [1]
   }
   n27 = "Transform (Modify) Bone - Bone: base_skirt_c_01"!(ComponentPose=@n26, bAlphaBoolEnabled:bool=True, Translation="0,0,0", Rotation="-10,0,0", Scale="1,1,1") -> (Pose) {
    alphaInputType=Bool
    rotationMode=BMM_Additive
    rotationSpace=BCS_BoneSpace
    translationMode=BMM_Additive
    translationSpace=BCS_BoneSpace
    alphaBoolBlend {blendInTime=0.2, blendOutTime=0.2}
    boneToModify {boneName=base_skirt_c_01}
    rotation {pitch=-10}
   }
   n28 = "Transform (Modify) Bone - Bone: dyn_skirtBack_c_02"!(ComponentPose=@n27, bAlphaBoolEnabled:bool=True, Translation="0,0,0", Rotation="-10,0,0", Scale="1,1,1") -> (Pose) {
    alphaInputType=Bool
    rotationMode=BMM_Additive
    rotationSpace=BCS_BoneSpace
    translationMode=BMM_Additive
    translationSpace=BCS_BoneSpace
    alphaBoolBlend {blendInTime=0.2, blendOutTime=0.2}
    boneToModify {boneName=dyn_skirtBack_c_02}
    rotation {pitch=-10}
   }
   Output Pose(Result=Component To Local(ComponentPose=@n28)) {layerGroup=None}
  ]}
 EventGraph {nodes [
   Event BlueprintUpdateAnimation()
   > n1 = Sequence!(execute) -> (then_0, then_1, then_2:Exec, then_3:Exec, then_4:Exec)
   n2 = Branch(execute=@n1.then_1, Condition=OR Boolean(A=Get bIsMoving2D(), B=Get bIsPlayingMeleeAnim()))
   > Set backSkirt(backSkirt:double=1) -> (Output_Get:double)
   Set backSkirt(execute=@n2.else, backSkirt=0)
   n3 = Branch(execute=@n1.then_2, Condition=NOT Boolean(A=Get bIsMoving2D()))
   > Set gluteOn(gluteOn:double=0) -> (Output_Get:double)
   Set gluteOn(execute=@n3.else, gluteOn=1)
   n4 = Branch(execute=@n1.then_3, Condition=Get bIsSkyDiving())
   > Set skydiving_var(skydiving_var:double=0) -> (Output_Get:double)
   Set skydiving_var(execute=@n4.else, skydiving_var=1)
   n5 = Branch(execute=@n1.then_4, Condition=OR Boolean(A=Get bIsJumping(), B=Get bIsMoving2D()))
   > Set jump_skirt(jump_skirt:double=1) -> (Output_Get:double)
   Set jump_skirt(execute=@n5.else, jump_skirt=0)
   n6 = "Gravity Override State Function (Parameter Asset) (Nativized)"(
    execute=@n1.then_0
    CharacterPartAnimInstance=Self-Reference()
    GameTimeInSeconds=GetGameTimeInSeconds()
    Parameters=&/M_MED_BullKeyNote_Parameters
   )
   > Set GravityOverride(GravityOverride=@n6)
   > Set RBAN Settings Out(RBAN Settings Out=@n6.OutRigidBodyAnimNodeParameters)
   > Set NotEmote(NotEmote:bool=NOT Boolean(A=Get bIsPlayingEmote())) -> (Output_Get:bool)
   n7 = Get bIsBaseballSliding()
   > Set BBSlide(BBSlide:double="To Float (Boolean)"(InBool=@n7)) -> (Output_Get:double)
   n8 = Get bIsKneeSliding()
   > Set Slidfe(Slidfe:double="To Float (Boolean)"(InBool=@n8)) -> (Output_Get:double)
   n9 = OR Boolean(A=@n8, B=@n7)
   > "Set not sliding "("not sliding ":double="To Float (Boolean)"(InBool=NOT Boolean(A=@n9))) -> (Output_Get:double)
   > Set Melee(Melee:double="To Float (Boolean)"(InBool=Get bIsPlayingMeleeAnim())) -> (Output_Get:double)
   n10 = AND Boolean!(
    A=Get bIsInFrontEnd()
    B=NOT Boolean(A=Get bIsPlayingEmote())
    C:bool=NOT Boolean(A=Get bIsPreviewingItemInFrontEnd() -> (bIsPreviewingItemInFrontEnd:bool))
   ) -> (ReturnValue)
   > Set FE(FE:double="To Float (Boolean)"(InBool=NOT Boolean(A=@n10))) -> (Output_Get:double)
   n11 = NOT Boolean(A=Get bIsPreviewingItemInFrontEnd())
   n12 = NOT Boolean(A=Get bIsPlayingEmote())
   n13 = Get bIsInFrontEnd()
   > Set FrontEnd(FrontEnd:double="To Float (Boolean)"(InBool=AND Boolean(A=@n13, B=@n12, C=@n11))) -> (Output_Get:double)
   n14 = Branch(Condition=Get bIsSkydiveFloating())
   > Set gravity(gravity:Vector="0,0,980") -> (Output_Get:Vector)
   > Set skirt_floatingVar(skirt_floatingVar:double=1) -> (Output_Get:double)
   Set gravity(execute=@n14.else, gravity="0,0,-980")
   > Set skirt_floatingVar(skirt_floatingVar=0)
   Get bIsSkydiveDiveMode()
  ]}
]
variables [
 backSkirt {type=number}
 BBSlide {type=number}
 FE {type=number}
 FrontEnd {type=number}
 gluteOn {type=number}
 gravity {type=Vector}
 GravityOverride {type=Vector}
 jump_skirt {type=number}
 Melee {type=number}
 "not sliding " {type=number}
 NotEmote {type=boolean}
 RBAN Settings Out {type=FortRigidBodyAnimNodeParameters}
 Set Gravity {type=Vector}
 SetAmplitud {type=Vector}
 SetFrequency {type=Rotator}
 skirt_floatingVar {type=number}
 skydiving_var {type=number}
 Slidfe {type=number}
]
