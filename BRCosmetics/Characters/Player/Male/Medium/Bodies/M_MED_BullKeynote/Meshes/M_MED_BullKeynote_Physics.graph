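# Physics asset for the M_MED_BullKeynote male medium player body, built around a simulated skirt.
# Eight front, side and back skirt chains each link three capsule bones to a base bone, which joins the pelvis with free rotation.
# Chain joints act as narrow hinges, allowing only slight swing on one axis with no twist.
# Capsules on pelvis, spine, forearms, hands, thighs and calves supply a simplified body alongside the skirt.
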PhysicsAsset /BRCosmetics/Characters/Player/Male/Medium/Bodies/M_MED_BullKeynote/Meshes/M_MED_BullKeynote_Physics
bodies {
 base_skirtBack_L [{shapeType=Capsule, length=0.505, radius=0.505, shapeName=base_skirtBack_L_capsule, rotation=<rot 0 -90 0>}]
 base_skirtBack_R [{shapeType=Capsule, length=16.505, radius=0.505, shapeName=base_skirtBack_R_capsule, rotation=<rot 0 -90 0>}]
 base_skirtBack_c_01 [{shapeType=Capsule, length=16.505, radius=0.505, shapeName=base_skirtBack_c_01_capsule, rotation=<rot 0 -90 0>}]
 base_skirtSide_L [{shapeType=Capsule, length=16.505, radius=0.505, shapeName=base_skirtSide_L_capsule, rotation=<rot 0 -90 0>}]
 base_skirtSide_R [{shapeType=Capsule, length=16.505, radius=0.505, shapeName=base_skirtSide_R_capsule, rotation=<rot 0 -90 0>}]
 base_skirt_L [{shapeType=Capsule, length=16.505, radius=0.505, shapeName=base_skirt_L_capsule, rotation=<rot 0 -90 0>}]
 base_skirt_R [{shapeType=Capsule, length=16.505, radius=0.505, shapeName=base_skirt_R_capsule, rotation=<rot 0 -90 0>}]
 base_skirt_c_01 [{shapeType=Capsule, length=16.505, radius=0.505, shapeName=base_skirt_c_01_capsule, rotation=<rot 0 -90 0>}]
 calf_l [{
   shapeType=Capsule
   length=37.8973
   radius=6.9957
   shapeName=calf_l_capsule
   center=<-27.577 -0.8051 2.0214>
   rotation=<rot -1.542 -90.0906 -86.4033>
  }]
 calf_r [{
   shapeType=Capsule
   length=39.1145
   radius=6.5924
   shapeName=calf_r_capsule
   center=<26.4189 1.3023 -0.9593>
   rotation=<rot -2.134 -91.6359 -88.3936>
  }]
 dyn_skirtBack_L_01 [{shapeType=Capsule, length=21, radius=3, center=<9.543 0 0>, rotation=<rot 90 90 0>}]
 dyn_skirtBack_L_02 [{shapeType=Capsule, length=37, radius=4, shapeName=dyn_skirtBack_L_02_capsule, center=<8.589 0 1.4622>, rotation=<rot 90 90 0>}]
 dyn_skirtBack_L_03 [{shapeType=Capsule, length=37, radius=4, shapeName=dyn_skirtBack_L_03_capsule, center=<8.535 0 1.2454>, rotation=<rot 90 90 0>}]
 dyn_skirtBack_R_01 [{shapeType=Capsule, length=21, radius=3, center=<-8.8387 0 0>, rotation=<rot 90 90 0>}]
 dyn_skirtBack_R_02 [{shapeType=Capsule, length=36, radius=5, shapeName=dyn_skirtBack_R_02_capsule, center=<-7.9045 0 -0.8951>, rotation=<rot 90 90 0>}]
 dyn_skirtBack_R_03 [{shapeType=Capsule, length=37, radius=4, shapeName=dyn_skirtBack_R_03_capsule, center=<-7.6991 0 -0.8951>, rotation=<rot 90 90 0>}]
 dyn_skirtBack_c_01 [{shapeType=Capsule, length=21, radius=3.5, center=<9.7358 0 0>, rotation=<rot 90 90 0>}]
 dyn_skirtBack_c_02 [{shapeType=Capsule, length=37, radius=4, shapeName=dyn_skirtBack_c_02_capsule, center=<9.0312 0 0>, rotation=<rot 90 90 0>}]
 dyn_skirtBack_c_03 [{shapeType=Capsule, length=37, radius=4, shapeName=dyn_skirtBack_c_03_capsule, center=<8.6175 0 0>, rotation=<rot 90 90 0>}]
 dyn_skirtSide_L_01 [{shapeType=Capsule, length=21, radius=3.5, center=<10.1801 0 0>, rotation=<rot 90 90 0>}]
 dyn_skirtSide_L_02 [{shapeType=Capsule, length=37, radius=4, shapeName=dyn_skirtSide_L_02_capsule, center=<7.6594 0 0>, rotation=<rot 90 90 0>}]
 dyn_skirtSide_L_03 [{shapeType=Capsule, length=37.5, radius=3.5, shapeName=dyn_skirtSide_L_03_capsule, center=<8.8477 0 1.0306>, rotation=<rot 90 90 0>}]
 dyn_skirtSide_R_01 [{shapeType=Capsule, length=21, radius=3, center=<-9.369 0 0>, rotation=<rot 90 90 0>}]
 dyn_skirtSide_R_02 [{shapeType=Capsule, length=36, radius=5, shapeName=dyn_skirtSide_R_02_capsule, center=<-6.977 0 -2.1396>, rotation=<rot 90 90 0>}]
 dyn_skirtSide_R_03 [{shapeType=Capsule, length=37, radius=4, shapeName=dyn_skirtSide_R_03_capsule, center=<-7.862 0 -2.1396>, rotation=<rot 90 90 0>}]
 dyn_skirt_L_01 [{shapeType=Capsule, length=21, radius=4.5, center=<11.1842 0 0>, rotation=<rot 90 90 0>}]
 dyn_skirt_L_02 [{shapeType=Capsule, length=37, radius=4.5, shapeName=dyn_skirt_L_02_capsule, center=<7.6059 0 1>, rotation=<rot 90 90 0>}]
 dyn_skirt_L_03 [{shapeType=Capsule, length=37, radius=4.5, shapeName=dyn_skirt_L_03_capsule, center=<9.2433 0 1.2>, rotation=<rot 90 90 0>}]
 dyn_skirt_R_01 [{shapeType=Capsule, length=21, radius=3.5, center=<-10.6671 0 0>, rotation=<rot 90 90 0>}]
 dyn_skirt_R_02 [{shapeType=Capsule, length=36, radius=5, shapeName=dyn_skirt_R_02_capsule, center=<-6.9727 0 -2.0275>, rotation=<rot 90 90 0>}]
 dyn_skirt_R_03 [{shapeType=Capsule, length=37, radius=4, shapeName=dyn_skirt_R_03_capsule, center=<-8.2354 0 -2.0275>, rotation=<rot 90 90 0>}]
 dyn_skirt_c_01 [{shapeType=Capsule, length=21, radius=4, center=<11.54 0 0>, rotation=<rot 90 90 0>}]
 dyn_skirt_c_02 [{shapeType=Capsule, length=37, radius=4, shapeName=dyn_skirt_c_02_capsule, center=<6.559 0 0>, rotation=<rot 90 90 0>}]
 dyn_skirt_c_03 [{shapeType=Capsule, length=37, radius=4, shapeName=dyn_skirt_c_03_capsule, center=<8.2396 0 0>, rotation=<rot 90 90 0>}]
 hand_l [{shapeType=Capsule, length=8.7934, radius=5.5084, center=<8.0332 -2.5741 -0.2658>, rotation=<rot 0 81.2453 -76.3461>}]
 hand_r [{shapeType=Capsule, length=8.8136, radius=5.5079, center=<-8.0147 2.5711 0.2707>, rotation=<rot 0 81.2387 -76.3593>}]
 lowerarm_l [{shapeType=Capsule, length=11.376, radius=4.6227, center=<12.1924 0.1441 -0.0168>, rotation=<rot 0 -90.3402 -89.1558>}]
 lowerarm_r [{shapeType=Capsule, length=11.3324, radius=4.6021, center=<-12.2313 -0.1782 0.0265>, rotation=<rot 0 -90.6108 -89.2979>}]
 pelvis [{shapeType=Capsule, length=17.3556, radius=8.4332, center=<7.4792 0.6148 0.6791>, rotation=<rot 0 -90 0>}]
 spine_02 [{shapeType=Capsule, length=12.3413, radius=8.6866, center=<9.2478 1.8989 -0.0563>, rotation=<rot 6.6164 -90 -0.5562>}]
 thigh_l [{shapeType=Capsule, length=44.4999, radius=5.9918, center=<-18.7383 0.5655 2.0493>, rotation=<rot 0 -87.7906 -88.2031>}]
 thigh_r [{shapeType=Capsule, length=39.0063, radius=5.5764, center=<19.1958 -0.7954 -1.4639>, rotation=<rot 0 -87.77 -88.1339>}]
}
constraints [
 {bone1Name=spine_02, bone2Name=pelvis, swing1LimitDegrees=45, swing2LimitDegrees=45, twistLimitDegrees=45}
 {bone1Name=lowerarm_l, bone2Name=spine_02, swing1LimitDegrees=45, swing2LimitDegrees=45, twistLimitDegrees=45}
 {bone1Name=hand_l, bone2Name=lowerarm_l, swing1LimitDegrees=45, swing2LimitDegrees=45, twistLimitDegrees=45}
 {bone1Name=lowerarm_r, bone2Name=spine_02, swing1LimitDegrees=45, swing2LimitDegrees=45, twistLimitDegrees=45}
 {bone1Name=hand_r, bone2Name=lowerarm_r, swing1LimitDegrees=45, swing2LimitDegrees=45, twistLimitDegrees=45}
 {bone1Name=thigh_l, bone2Name=pelvis, swing1LimitDegrees=45, swing2LimitDegrees=45, twistLimitDegrees=45}
 {bone1Name=thigh_r, bone2Name=pelvis, swing1LimitDegrees=45, swing2LimitDegrees=45, twistLimitDegrees=45}
 {bone1Name=dyn_skirtBack_L_01, bone2Name=base_skirtBack_L, swing1LimitDegrees=0, swing2LimitDegrees=20, twistLimitDegrees=0}
 {bone1Name=dyn_skirtSide_L_01, bone2Name=base_skirtSide_L, swing1LimitDegrees=0, swing2LimitDegrees=20, twistLimitDegrees=0}
 {bone1Name=dyn_skirt_L_01, bone2Name=base_skirt_L, swing1LimitDegrees=0, swing2LimitDegrees=20, twistLimitDegrees=0}
 {bone1Name=dyn_skirtBack_R_01, bone2Name=base_skirtBack_R, swing1LimitDegrees=0, swing2LimitDegrees=20, twistLimitDegrees=0}
 {bone1Name=dyn_skirtSide_R_01, bone2Name=base_skirtSide_R, swing1LimitDegrees=0, swing2LimitDegrees=20, twistLimitDegrees=0}
 {bone1Name=dyn_skirt_R_01, bone2Name=base_skirt_R, swing1LimitDegrees=0, swing2LimitDegrees=20, twistLimitDegrees=0}
 {bone1Name=dyn_skirt_c_01, bone2Name=base_skirt_c_01, swing1LimitDegrees=0, swing2LimitDegrees=20, twistLimitDegrees=0}
 {bone1Name=dyn_skirtBack_c_01, bone2Name=base_skirtBack_c_01, swing1LimitDegrees=0, swing2LimitDegrees=20, twistLimitDegrees=0}
 {
  bone1Name=base_skirtBack_L
  bone2Name=pelvis
  swing1LimitDegrees=45
  swing1Motion=Free
  swing2LimitDegrees=45
  swing2Motion=Free
  twistLimitDegrees=45
  twistMotion=Free
 }
 {bone1Name=dyn_skirtBack_L_02, bone2Name=dyn_skirtBack_L_01, swing1LimitDegrees=0, swing2LimitDegrees=20, twistLimitDegrees=0}
 {bone1Name=dyn_skirtBack_L_03, bone2Name=dyn_skirtBack_L_02, swing1LimitDegrees=0, swing2LimitDegrees=20, twistLimitDegrees=0}
 {
  bone1Name=base_skirtBack_R
  bone2Name=pelvis
  swing1LimitDegrees=45
  swing1Motion=Free
  swing2LimitDegrees=45
  swing2Motion=Free
  twistLimitDegrees=45
  twistMotion=Free
 }
 {bone1Name=dyn_skirtBack_R_02, bone2Name=dyn_skirtBack_R_01, swing1LimitDegrees=0, swing2LimitDegrees=20, twistLimitDegrees=0}
 {bone1Name=dyn_skirtBack_R_03, bone2Name=dyn_skirtBack_R_02, swing1LimitDegrees=0, swing2LimitDegrees=20, twistLimitDegrees=0}
 {
  bone1Name=base_skirtBack_c_01
  bone2Name=pelvis
  swing1LimitDegrees=45
  swing1Motion=Free
  swing2LimitDegrees=45
  swing2Motion=Free
  twistLimitDegrees=45
  twistMotion=Free
 }
 {bone1Name=dyn_skirtBack_c_02, bone2Name=dyn_skirtBack_c_01, swing1LimitDegrees=0, swing2LimitDegrees=20, twistLimitDegrees=0}
 {bone1Name=dyn_skirtBack_c_03, bone2Name=dyn_skirtBack_c_02, swing1LimitDegrees=0, swing2LimitDegrees=20, twistLimitDegrees=0}
 {
  bone1Name=base_skirtSide_L
  bone2Name=pelvis
  swing1LimitDegrees=45
  swing1Motion=Free
  swing2LimitDegrees=45
  swing2Motion=Free
  twistLimitDegrees=45
  twistMotion=Free
 }
 {bone1Name=dyn_skirtSide_L_02, bone2Name=dyn_skirtSide_L_01, swing1LimitDegrees=0, swing2LimitDegrees=20, twistLimitDegrees=0}
 {bone1Name=dyn_skirtSide_L_03, bone2Name=dyn_skirtSide_L_02, swing1LimitDegrees=0, swing2LimitDegrees=20, twistLimitDegrees=0}
 {
  bone1Name=base_skirtSide_R
  bone2Name=pelvis
  swing1LimitDegrees=45
  swing1Motion=Free
  swing2LimitDegrees=45
  swing2Motion=Free
  twistLimitDegrees=45
  twistMotion=Free
 }
 {bone1Name=dyn_skirtSide_R_02, bone2Name=dyn_skirtSide_R_01, swing1LimitDegrees=0, swing2LimitDegrees=20, twistLimitDegrees=0}
 {bone1Name=dyn_skirtSide_R_03, bone2Name=dyn_skirtSide_R_02, swing1LimitDegrees=0, swing2LimitDegrees=20, twistLimitDegrees=0}
 {
  bone1Name=base_skirt_L
  bone2Name=pelvis
  swing1LimitDegrees=45
  swing1Motion=Free
  swing2LimitDegrees=45
  swing2Motion=Free
  twistLimitDegrees=45
  twistMotion=Free
 }
 {bone1Name=dyn_skirt_L_02, bone2Name=dyn_skirt_L_01, swing1LimitDegrees=0, swing2LimitDegrees=20, twistLimitDegrees=0}
 {bone1Name=dyn_skirt_L_03, bone2Name=dyn_skirt_L_02, swing1LimitDegrees=0, swing2LimitDegrees=20, twistLimitDegrees=0}
 {
  bone1Name=base_skirt_R
  bone2Name=pelvis
  swing1LimitDegrees=45
  swing1Motion=Free
  swing2LimitDegrees=45
  swing2Motion=Free
  twistLimitDegrees=45
  twistMotion=Free
 }
 {bone1Name=dyn_skirt_R_02, bone2Name=dyn_skirt_R_01, swing1LimitDegrees=0, swing2LimitDegrees=20, twistLimitDegrees=0}
 {bone1Name=dyn_skirt_R_03, bone2Name=dyn_skirt_R_02, swing1LimitDegrees=0, swing2LimitDegrees=20, twistLimitDegrees=0}
 {
  bone1Name=base_skirt_c_01
  bone2Name=pelvis
  swing1LimitDegrees=45
  swing1Motion=Free
  swing2LimitDegrees=45
  swing2Motion=Free
  twistLimitDegrees=45
  twistMotion=Free
 }
 {bone1Name=dyn_skirt_c_02, bone2Name=dyn_skirt_c_01, swing1LimitDegrees=0, swing2LimitDegrees=20, twistLimitDegrees=0}
 {bone1Name=dyn_skirt_c_03, bone2Name=dyn_skirt_c_02, swing1LimitDegrees=0, swing2LimitDegrees=20, twistLimitDegrees=0}
 {
  bone1Name=calf_l
  bone2Name=thigh_l
  swing1LimitDegrees=45
  swing1Motion=Free
  swing2LimitDegrees=45
  swing2Motion=Free
  twistLimitDegrees=45
  twistMotion=Free
 }
 {
  bone1Name=calf_r
  bone2Name=thigh_r
  swing1LimitDegrees=45
  swing1Motion=Free
  swing2LimitDegrees=45
  swing2Motion=Free
  twistLimitDegrees=45
  twistMotion=Free
 }
]
