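# Physics asset for the M_MED_BullKeynote male medium player body cosmetic, covering the lower body and front and back skirt.
# Capsules cover the pelvis, thighs, calves and feet, plus two three-segment dyn_skirt chains, each starting from a thin base capsule.
# Skirt bases, knees and ankles rotate freely, while hips are limited.
# Skirt segments have tight swing limits and zero twist, with the front chain restricted to one swing axis.
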PhysicsAsset /BRCosmetics/Characters/Player/Male/Medium/Bodies/M_MED_BullKeynote/Meshes/M_MED_BullKeynotec_Physics
bodies {
 base_skirtBack_c_01 [{shapeType=Capsule, length=16.505, radius=0.505, shapeName=base_skirtBack_c_01_capsule, rotation=<rot 0 -90 0>}]
 base_skirt_c_01 [{shapeType=Capsule, length=16.505, radius=0.505, shapeName=base_skirt_c_01_capsule, rotation=<rot 0 -90 0>}]
 calf_l [{shapeType=Capsule, length=32.3707, radius=5.2892, shapeName=calf_l_capsule, center=<-20.0984 -1.7648 0>, rotation=<rot 90 0 0>}]
 calf_r [{shapeType=Capsule, length=34.3944, radius=5.3762, shapeName=calf_r_capsule, center=<20.8865 1.4174 0>, rotation=<rot 90 0 0>}]
 dyn_skirtBack_c_01 [{shapeType=Capsule, length=20, radius=4.5, center=<9.7358 0 0>, rotation=<rot 90 90 0>}]
 dyn_skirtBack_c_02 [{shapeType=Capsule, length=35.5, radius=5.5, shapeName=dyn_skirtBack_c_02_capsule, center=<9.0312 0 0>, rotation=<rot 90 90 0>}]
 dyn_skirtBack_c_03 [{shapeType=Capsule, length=6, radius=5, shapeName=dyn_skirtBack_c_03_capsule, center=<7.1388 0 0>, rotation=<rot 90 90 0>}]
 dyn_skirt_c_01 [{shapeType=Capsule, length=21, radius=4, center=<11.54 0 0>, rotation=<rot 0 0 80>}]
 dyn_skirt_c_02 [{shapeType=Capsule, length=37, radius=4, shapeName=dyn_skirt_c_02_capsule, center=<6.559 0 0>, rotation=<rot 0 0 80>}]
 dyn_skirt_c_03 [{shapeType=Capsule, length=37, radius=4, shapeName=dyn_skirt_c_03_capsule, center=<8.2396 0 0>, rotation=<rot 0 0 76>}]
 foot_l [{shapeType=Capsule, length=9.9076, radius=6.91, shapeName=foot_l_capsule, center=<-4.5886 6.0664 0.4443>, rotation=<rot 0 11.2405 -88.4651>}]
 foot_r [{shapeType=Capsule, length=9.9831, radius=6.9581, shapeName=foot_r_capsule, center=<4.857 -6.8229 -0.1767>, rotation=<rot 0 12.2368 -89.3043>}]
 pelvis [{shapeType=Capsule, length=17.3556, radius=8.4332, center=<7.4792 0.6148 0.6791>, rotation=<rot 0 -90 0>}]
 thigh_l [{shapeType=Capsule, length=44.4999, radius=5.9918, center=<-18.7383 0.5655 2.0493>, rotation=<rot 0 -87.7906 -88.2031>}]
 thigh_r [{shapeType=Capsule, length=39.0063, radius=5.5764, center=<19.1958 -0.7954 -1.4639>, rotation=<rot 0 -87.77 -88.1339>}]
}
constraints [
 {bone1Name=thigh_l, bone2Name=pelvis, swing1LimitDegrees=45, swing2LimitDegrees=45, twistLimitDegrees=45}
 {bone1Name=thigh_r, bone2Name=pelvis, swing1LimitDegrees=45, swing2LimitDegrees=45, twistLimitDegrees=45}
 {bone1Name=dyn_skirt_c_01, bone2Name=base_skirt_c_01, swing1LimitDegrees=0, swing2LimitDegrees=10, twistLimitDegrees=0}
 {bone1Name=dyn_skirtBack_c_01, bone2Name=base_skirtBack_c_01, swing1LimitDegrees=5, swing2LimitDegrees=10, twistLimitDegrees=0}
 {
  bone1Name=base_skirtBack_c_01
  bone2Name=pelvis
  swing1LimitDegrees=45
  swing1Motion=Free
  swing2LimitDegrees=45
  swing2Motion=Free
  twistLimitDegrees=45
  twistMotion=Free
 }
 {bone1Name=dyn_skirtBack_c_02, bone2Name=dyn_skirtBack_c_01, swing1LimitDegrees=5, swing2LimitDegrees=20, twistLimitDegrees=0}
 {bone1Name=dyn_skirtBack_c_03, bone2Name=dyn_skirtBack_c_02, swing1LimitDegrees=5, swing2LimitDegrees=20, twistLimitDegrees=0}
 {
  bone1Name=base_skirt_c_01
  bone2Name=pelvis
  swing1LimitDegrees=45
  swing1Motion=Free
  swing2LimitDegrees=45
  swing2Motion=Free
  twistLimitDegrees=45
  twistMotion=Free
 }
 {bone1Name=dyn_skirt_c_02, bone2Name=dyn_skirt_c_01, swing1LimitDegrees=0, swing2LimitDegrees=20, twistLimitDegrees=0}
 {bone1Name=dyn_skirt_c_03, bone2Name=dyn_skirt_c_02, swing1LimitDegrees=0, swing2LimitDegrees=20, twistLimitDegrees=0}
 {
  bone1Name=foot_r
  bone2Name=calf_r
  swing1LimitDegrees=45
  swing1Motion=Free
  swing2LimitDegrees=45
  swing2Motion=Free
  twistLimitDegrees=45
  twistMotion=Free
 }
 {
  bone1Name=foot_l
  bone2Name=calf_l
  swing1LimitDegrees=45
  swing1Motion=Free
  swing2LimitDegrees=45
  swing2Motion=Free
  twistLimitDegrees=45
  twistMotion=Free
 }
 {
  bone1Name=calf_l
  bone2Name=thigh_l
  swing1LimitDegrees=45
  swing1Motion=Free
  swing2LimitDegrees=45
  swing2Motion=Free
  twistLimitDegrees=45
  twistMotion=Free
 }
 {
  bone1Name=calf_r
  bone2Name=thigh_r
  swing1LimitDegrees=45
  swing1Motion=Free
  swing2LimitDegrees=45
  swing2Motion=Free
  twistLimitDegrees=45
  twistMotion=Free
 }
]
